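# Physics asset for the Cardboard DrumKit instrument cosmetic, giving bodies and joint limits to its box flaps and cymbal pizza bones.
# Four small pizza capsules form a chain with locked twist and moderate swing.
# The left tom flap, the only box shape, hangs off tom_shake_l with wide swing on one axis.
# The cymbal flap joins crash_cymbal_r with twist and one swing locked, leaving a slight single-axis wobble.
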PhysicsAsset /SparksCosmetics/Instrument/DrumKit/Cardboard/Meshes/Cardboard_DrumKit_Physics
bodies {
 L_tom_box_flap_01 [{
   shapeType=Box
   extentX=0.73
   extentY=12.5
   extentZ=21.8526
   shapeName=L_tom_box_flap_01_box
   center=<-0.4377 0.0393 7.569>
   rotation=<rot 0 -85.1391 -89.1127>
  }]
 crash_cymbal_r [{shapeType=Capsule, length=0, radius=2}]
 cymbal_box_flap_01 [{shapeType=Capsule, length=0, radius=2, shapeName=cymbal_box_flap_01_capsule, center=<-0.6567 -10.8515 0>}]
 cymbal_pizza_01 [{shapeType=Capsule, length=1.5, radius=1, rotation=<rot 67.7315 25.505 27.2726>}]
 cymbal_pizza_02 [{shapeType=Capsule, length=1.5, radius=1}]
 cymbal_pizza_03 [{shapeType=Capsule, length=1.5, radius=1, center=<0 0 -0.2614>}]
 cymbal_pizza_04 [{shapeType=Capsule, length=1.5, radius=1, center=<0 0 -0.4249>}]
 tom_shake_l [{shapeType=Capsule, length=1, radius=4, shapeName=tom_shake_l_capsule}]
}
constraints [
 {bone1Name=cymbal_pizza_02, bone2Name=cymbal_pizza_01, swing1LimitDegrees=25, swing2LimitDegrees=30, twistLimitDegrees=0, twistMotion=Locked}
 {
  bone1Name=cymbal_pizza_03
  bone2Name=cymbal_pizza_02
  swing1LimitDegrees=34.48
  swing2LimitDegrees=35
  twistLimitDegrees=0
  twistMotion=Locked
 }
 {bone1Name=cymbal_pizza_04, bone2Name=cymbal_pizza_03, swing1LimitDegrees=30, swing2LimitDegrees=35, twistLimitDegrees=0, twistMotion=Locked}
 {bone1Name=L_tom_box_flap_01, bone2Name=tom_shake_l, swing1LimitDegrees=120, swing2LimitDegrees=0, twistLimitDegrees=0}
 {
  bone1Name=cymbal_box_flap_01
  bone2Name=crash_cymbal_r
  swing1LimitDegrees=0
  swing1Motion=Locked
  swing2LimitDegrees=5
  twistLimitDegrees=0
  twistMotion=Locked
 }
]
